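# Translucent unlit Quest skin material for MistRaven's wraith form, rendering a smoky body with depth test disabled.
# Emissive blends shadow, bottom and top colors using local height and half-Lambert sky light on panning world-aligned distortion normals.
# Opacity combines inverted Fresnel, depth fade and distorted world-aligned smoke, then passes through a translucency-sorting function.
# A switch adds cosmetic eye adaptation; the screen-space smoke WPO branch never reaches the material attributes.
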Material /BRCosmetics/Characters/Player/Female/Medium/Bodies/F_MED_MistRaven_Peck/Skins/Quest/Materials/M_MistRaven_WraithForm_Quest
graph {
 properties {
  allowTranslucentCustomDepthWrites=true
  bAutomaticallySetUsageInEditor=false
  bDisableDepthTest=true
  bUseMaterialAttributes=true
  bUsedWithMeshDeformer=true
  bUsedWithSkeletalMesh=true
  bUsedWithStaticLighting=true
  blendMode=BLEND_Translucent
  shadingModel=MSM_Unlit
  usedShadingModels=Unlit
 }
 nodes [
  n1 = Multiply(A=ScreenPosition().ViewportUV, B=ScalarParameter() {defaultValue=1.5, parameterName=SmokeWPO_Tiling, sliderMax=4})
  n2 = Panner(Coordinate=@n1) {speedX=-0.6, speedY=-0.5}
  n3 = Panner(Coordinate=@n1) {speedX=0.6, speedY=0.5}
  n4 = Add(
   A=Multiply(
    A=Constant3Vector() {constant=<rgba 1 0 0 0>}
    B=TextureSample(UVs=@n2) {texture=/Game/Global/Textures/Default/Tileable_Patterns/T_Blowing_Smoke_Pattern2_Tile_DISP}.R
   )
   B=Multiply(
    A=Constant3Vector() {constant=<rgba -1 0 0 0>}
    B=TextureSample(UVs=@n3) {texture=&T_Blowing_Smoke_Pattern2_Tile_DISP}.R
   )
  )
  n5 = ScalarParameter() {defaultValue=2, parameterName=WPO_Scale, sliderMax=10}
  NamedRerouteDeclaration(None=Transform(None=Multiply(A=@n4, B=@n5)) {transformSourceType=TRANSFORMSOURCE_View}) {name=WPO, nodeColor=<rgba 1 0 0.6094 1>}
  n6 = MaterialFunctionCall(Float2=@n1) {materialFunction=&BreakOutFloat2Components}
  n7 = Sine(None=@n6.G)
  n8 = Add(A=@n6.G, B=Multiply(A=Multiply(A=@n7, B=@n7)) {constB=0.1})
  n9 = Add(A=@n6.R) {constB=-0.5}
  n10 = Multiply(A=Power(Base=Multiply(A=@n9, B=@n9)) {constExponent=1.1}) {constB=4}
  n11 = Multiply(A=MaterialFunctionCall(X=@n6.R, Y=Add(A=@n8, B=@n10)) {materialFunction=&MakeFloat2}.Result) {constB=2}
  n12 = TextureSample(UVs=Panner(Coordinate=Multiply(A=@n11) {constB=0.2}) {speedX=0.05, speedY=-0.1}) {samplerType=SAMPLERTYPE_Normal, texture=/BRCosmetics/Gadgets/Assets/VinderTech_GliderChute/Glider_Genius_Blob/VFX/T_CloudBubble_Normal}
  n13 = ComponentMask(None=@n12.RGB) {g=true, r=true}
  n14 = Add(A=Panner(Coordinate=@n11) {speedX=-0.1, speedY=-0.5}, B=Multiply(A=Add(A=@n13) {constB=-0.5}) {constB=0.2}) {constB=0}
  NamedRerouteDeclaration(None=Power(Base=TextureSample(UVs=@n14) {texture=&T_Blowing_Smoke_Pattern2_Tile_DISP}.R) {constExponent=0.4545}) {name=SmokeDistorted, nodeColor=<rgba 0 0.3203 1 1>}
  NamedRerouteUsage()
  n15 = VectorParameter() {parameterName=Color_Shadow, defaultValue=<rgba 0.01 0.01 0.01 0>}
  n16 = LinearInterpolate(
   A=VectorParameter() {parameterName=Color_Bottom, defaultValue=<rgba 0.0806 0.0124 0.1281 1>}.RGB
   B=VectorParameter() {parameterName=Color_Top, defaultValue=<rgba 0.2573 0.0228 0.0228 1>}.RGB
   Alpha=Power(Base=NamedRerouteUsage())
  )
  n17 = LocalPosition() {localOrigin=InstancePreSkinning}
  n18 = Saturate(None=VertexInterpolator(VS=Divide(A=@n17.Z) {constB=200}).PS)
  n19 = ScalarParameter() {defaultValue=60, parameterName=Scale}
  n20 = Constant3Vector() {constant=<rgba -0.08 0.08 -0.6 0>}
  n21 = Multiply(A=Time(), B=ScalarParameter() {defaultValue=10, parameterName=PanSpeed})
  n22 = Subtract(A=WorldPosition().XYZ, B=Multiply(A=ActorPositionWS()) {constB=0.95})
  n23 = MaterialFunctionCall(
   TextureObject=TextureObject() {samplerType=SAMPLERTYPE_Normal, texture=/Game/Global/Textures/Default/Tileable_Patterns/T_Tilling_Distortion_Map_01_n}
   TextureSize=Add(A=@n19) {constB=10}
   WorldPosition=Add(A=Multiply(A=@n20, B=@n21), B=@n22)
  ) {materialFunction=&WorldAlignedNormal}
  n24 = Transform(None=MaterialFunctionCall(
    BaseNormal=TextureSampleParameter2D() {parameterName=Normal, samplerType=SAMPLERTYPE_Normal, texture=&FlatNormalBlue}.RGB
    AdditionalNormal=@n23.XYZ Texture
   ) {materialFunction=&BlendAngleCorrectedNormals}.Result)
  n25 = DotProduct(A=SkyAtmosphereLightDirection(), B=@n24)
  n26 = Saturate(None=Add(A=Multiply(A=@n25) {constB=0.5}) {constB=0.5})
  n27 = Fresnel(Normal=@n24) {baseReflectFraction=0, exponent=1}
  n28 = DepthFade(Opacity=NamedRerouteDeclaration(None=OneMinus(None=@n27)) {name=Fresnel, nodeColor=<rgba 1 0.0938 0 1>}) {fadeDistanceDefault=20}
  n29 = Add(A=@n23.XY Texture) {constB=-0.5}
  n30 = Multiply(A=@n21, B=Multiply(A=Constant3Vector() {constant=<rgba -0.08 0.08 -0.6 0>}) {constB=0.5})
  n31 = MaterialFunctionCall(
   TextureObject=TextureObject() {texture=/Game/Global/Textures/Default/Tileable_Patterns/T_Smoke_OUT}
   TextureSize=@n19
   WorldPosition=Add(A=Multiply(A=@n29) {constB=2}, B=Add(A=@n30, B=@n22))
  ) {materialFunction=&WorldAlignedTexture}
  n32 = LinearInterpolate(
   A=MaterialFunctionCall(Input=ComponentMask(None=@n31.XYZ Texture) {r=true}, Target Low=Constant() {r=1}, Target High=Constant() {r=2}) {materialFunction=&RemapValueRange}.Result
   Alpha=@n26
  ) {constB=1.1}
  n33 = SetMaterialAttributes!(
   MaterialAttributes
   Emissive Color=LinearInterpolate(A=@n15.RGB, B=LinearInterpolate(A=@n15.RGB, B=@n16, Alpha=@n18), Alpha=@n26)
   Opacity=MaterialFunctionCall!(Opacity=Multiply(A=@n28, B=@n32)) {materialFunction=/Game/Global/Materials/MaterialGym/Materials/SortedOutline/MF_SortTranslucency}.Result
  ) {attributeSetTypes [#1, #2]}
  n34 = MaterialFunctionCall!(
   Emissive
   Alpha=ScalarParameter() {defaultValue=1, group=Cosmetic Emissive EyeAdaptation, parameterName=Cosmetic Emissive EyeAdaptation Alpha}
   Material=@n33
  ) {materialFunction=/Game/Packages/Fortress_SharedMaterials/Base_Material_Functions/BaseCharacterShaderFunctions/MF_CosmeticEmissiveEyeAdaptation}
  StaticSwitchParameter(True=@n34.Material, False=@n33) {defaultValue=true, group=Cosmetic Emissive EyeAdaptation, parameterName=Cosmetic Emissive EyeAdaptation, sortPriority=100}
  VertexColor()
 ]
}
